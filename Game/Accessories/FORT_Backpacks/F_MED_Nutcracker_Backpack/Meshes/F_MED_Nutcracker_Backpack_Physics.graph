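# Physics asset for the Nutcracker backpack accessory, giving the figure's body, head, hat, shoulders and thighs simple capsule collision for physics-driven motion.
# The hat joins the head, the body joins the head, and shoulders and thighs join the body.
# Hat, shoulder and thigh joints lock twist and one swing axis, acting as single-axis hinges.
# Only the body-to-head joint leaves all three angular axes unlocked.
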PhysicsAsset /Game/Accessories/FORT_Backpacks/F_MED_Nutcracker_Backpack/Meshes/F_MED_Nutcracker_Backpack_Physics
bodies {
 body [{shapeType=Capsule, length=4.4996, radius=2.5526, center=<-1.1472 -0.3864 0>}]
 hat [{shapeType=Capsule, length=6.931, radius=4.93, center=<5.9864 1.6248 0.0771>, rotation=<rot 0 -75.2656 -89.9727>}]
 head [{shapeType=Capsule, length=0, radius=4, center=<3.304 0.0001 0>}]
 shoulder_l [{shapeType=Capsule, length=2, radius=1, center=<1.4999 0.4404 0>, rotation=<rot 90 0 0>}]
 shoulder_r [{shapeType=Capsule, length=2, radius=1, center=<1.5001 0.4404 0>, rotation=<rot 90 0 0>}]
 thigh_l [{shapeType=Capsule, length=5.724, radius=1.0292, center=<2.549 0.404 0.7069>, rotation=<rot 0 100.5843 -76.7309>}]
 thigh_r [{shapeType=Capsule, length=2.9737, radius=1.0293, center=<1.5189 0.2009 -0.6893>, rotation=<rot 0 -79.3675 -72.3243>}]
}
constraints [
 {bone1Name=hat, bone2Name=head, swing1LimitDegrees=10, swing2LimitDegrees=45, swing2Motion=Locked, twistLimitDegrees=45, twistMotion=Locked}
 {
  bone1Name=shoulder_l
  bone2Name=body
  swing1LimitDegrees=30
  swing2LimitDegrees=45
  swing2Motion=Locked
  twistLimitDegrees=45
  twistMotion=Locked
 }
 {
  bone1Name=shoulder_r
  bone2Name=body
  swing1LimitDegrees=30
  swing2LimitDegrees=45
  swing2Motion=Locked
  twistLimitDegrees=45
  twistMotion=Locked
 }
 {bone1Name=body, bone2Name=head, swing1LimitDegrees=45, swing2LimitDegrees=45, twistLimitDegrees=45}
 {
  bone1Name=thigh_r
  bone2Name=body
  swing1LimitDegrees=45
  swing1Motion=Locked
  swing2LimitDegrees=45
  twistLimitDegrees=45
  twistMotion=Locked
 }
 {
  bone1Name=thigh_l
  bone2Name=body
  swing1LimitDegrees=45
  swing1Motion=Locked
  swing2LimitDegrees=45
  twistLimitDegrees=45
  twistMotion=Locked
 }
]
